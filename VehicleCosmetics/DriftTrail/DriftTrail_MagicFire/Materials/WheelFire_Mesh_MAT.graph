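# Translucent, unlit, two-sided particle material for the wheel fire mesh in the MagicFire drift trail cosmetic.
# Two panning blob-noise samples are summed and tinted by particle color for emissive; the first swaps the TextureSpeed and TextureScale dynamic parameters relative to the second.
# Their red channels, shaped by UV edge falloffs, a third panning noise and inverted Fresnel, form an alpha mask.
# A dynamic Threshold is subtracted to erode it, then depth fade is applied.
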Material /VehicleCosmetics/DriftTrail/DriftTrail_MagicFire/Materials/WheelFire_Mesh_MAT
graph {
 properties {
  bAutomaticallySetUsageInEditor=false
  bUsedWithMeshDeformer=true
  bUsedWithNiagaraMeshParticles=true
  bUsedWithNiagaraRibbons=true
  bUsedWithNiagaraSprites=true
  bUsedWithSkeletalMesh=true
  blendMode=BLEND_Translucent
  pixelDepthOffsetMode=PDOM_Legacy
  shadingModel=MSM_Unlit
  twoSided=true
  usedShadingModels=Unlit
 }
 nodes [
  n1 = TextureCoordinate() {uTiling=4}
  n2 = DynamicParameter() {paramNames [TextureSpeed,TextureScale,Param3,Threshold]}
  n3 = Time()
  n4 = TextureSample(UVs=Panner(Coordinate=Multiply(A=@n2.TextureSpeed, B=@n1), Time=Multiply(A=@n3, B=@n2.TextureScale)) {speedX=0.2, speedY=-0.8}) {texture=/VehicleCosmetics/Boosts/Boost_MagicFire/Textures/T_VCVFX_boost_MagicFire_BlobNoise_Glow_T}
  n5 = TextureCoordinate()
  n6 = Time()
  n7 = TextureSample(UVs=Panner(Coordinate=Multiply(A=@n2.TextureScale, B=@n5), Time=Multiply(A=@n6, B=@n2.TextureSpeed)) {speedX=0.1, speedY=-0.5}) {texture=&T_VCVFX_boost_MagicFire_BlobNoise_Glow_T}
  n8 = Power(Base=ComponentMask(None=TextureCoordinate()) {r=true}, Exp=Constant() {r=2})
  n9 = Constant() {r=20}
  n10 = TextureCoordinate()
  n11 = Power(Base=OneMinus(None=ComponentMask(None=@n10) {r=true}), Exp=Constant() {r=2})
  n12 = Constant() {r=20}
  n13 = Power(Base=ComponentMask(None=TextureCoordinate()) {g=true}, Exp=Constant() {r=7})
  n14 = Constant() {r=10}
  n15 = TextureCoordinate()
  n16 = Power(Base=OneMinus(None=ComponentMask(None=@n15) {g=true}), Exp=Constant() {r=6})
  n17 = Constant() {r=1}
  n18 = Multiply(A=Multiply(A=Multiply(A=@n8, B=@n9), B=Multiply(A=@n11, B=@n12)), B=Multiply(A=Multiply(A=@n13, B=@n14), B=Multiply(A=@n16, B=@n17)))
  n19 = TextureCoordinate() {uTiling=5}
  n20 = OneMinus(None=ComponentMask(None=TextureCoordinate()) {g=true})
  n21 = Constant() {r=2}
  n22 = Multiply(
   A=TextureSample(UVs=Panner(Coordinate=@n19) {speedX=1, speedY=1}) {texture=&T_VCVFX_boost_MagicFire_BlobNoise_Glow_T}.RGB
   B=Multiply(A=Power(Base=@n20, Exp=@n21), B=Constant() {r=50})
  )
  n23 = OneMinus(None=Fresnel() {exponent=1})
  n24 = Constant() {r=6}
  n25 = Multiply(A=Multiply(A=Add(A=@n4.R, B=@n7.R), B=Multiply(A=@n18, B=@n22)), B=Multiply(A=Constant() {r=2}, B=Power(Base=@n23, Exp=@n24)))
  n26 = Constant() {r=1}
  MaterialExpressionConstantClamp_0 = Saturate(None=Multiply(
    A=Subtract(A=@n25, B=@n2.Threshold)
    B=MaterialFunctionCall(Opacity In=@n26) {materialFunction=&MF_VCVFX_DepthFade}.Result
   ))
  MaterialExpressionMeshEmitterVertexColor_0 = ParticleColor()
  MaterialFunctionCall(Emissive=Multiply(A=@MaterialExpressionMeshEmitterVertexColor_0.RGB, B=Add(A=@n4.RGB, B=@n7.RGB))) {materialFunction=&MF_FX_EmissiveAdjustments}
 ]
}
